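; Quest prop for the NOMS sign that reveals letters from the local player's quest progress and lights up on a final quest.
; After a five-second workaround delay and playlist load, clients check O, S, M, N quests in a nested chain, unhiding each completed letter.
; With all four done, the full sign shows, already lit or watching quest updates.
; Completion runs a timeline raising dynamic-material emissive, then plays success sound and FX.
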(event EventBeginPlay
  (Utilities|FlowControl|Sequence
    (:then_0
      ;[1 0.5445 0.1188] Workaround - needs to be something better for ship
      (Utilities|FlowControl|Delay 5.0)
      ;[1 1 1] CLIENT AND SERVER - Ensure playlist is loaded
      (bind _asfort_game_state_athena (Utilities|Casting|CastToFortGameStateAthena (Game|GetGameState)))
      (if (Variables|/Script/FortniteGame.FortGameStateAthena|GetbPlaylistDataIsLoaded _asfort_game_state_athena)
        ;[1 1 1] Set Visibility of Sign Letters Depending on Quest Progress
        (Networking|SwitchHasAuthority
          (:Remote
            (bind (_questisvalid _questcompleted) (Default|AthenaCheckQuestandObjectiveCompletion (Utilities|Array|Get(acopy) (Fortnite|GetLocalFortPlayerControllers true)) (Variables|QuestStuff|GetQuestItem_O) (Variables|QuestStuff|GetObjBackendName_O)))
            (if _questcompleted
              (Development|SetHiddeninGame (Variables|Default|GetSM_O_SignPiece) false false)
              (bind (_questisvalid_2 _questcompleted_2) (Default|AthenaCheckQuestandObjectiveCompletion (Utilities|Array|Get(acopy) (Fortnite|GetLocalFortPlayerControllers true)) (Variables|QuestStuff|GetQuestItem_S) (Variables|QuestStuff|GetObjBackendName_S)))
              (if _questcompleted_2
                (Development|SetHiddeninGame (Variables|Default|GetSM_S_SignPiece) false false)
                (bind (_questisvalid_3 _questcompleted_3) (Default|AthenaCheckQuestandObjectiveCompletion (Utilities|Array|Get(acopy) (Fortnite|GetLocalFortPlayerControllers true)) (Variables|QuestStuff|GetQuestItem_M) (Variables|QuestStuff|GetObjBackendName_M)))
                (if _questcompleted_3
                  (Development|SetHiddeninGame (Variables|Default|GetSM_M_SignPiece) false false)
                  (bind (_questisvalid_4 _questcompleted_4) (Default|AthenaCheckQuestandObjectiveCompletion (Utilities|Array|Get(acopy) (Fortnite|GetLocalFortPlayerControllers true)) (Variables|QuestStuff|GetQuestItem_N) (Variables|QuestStuff|GetObjBackendName_N)))
                  (if _questcompleted_4
                    (Development|SetHiddeninGame (Variables|Default|GetSM_N_SignPiece) false false)
                    (CallFunction|TestifFullSignVisible)
                    (else
                      _))
                  (else
                    _))
                (Variables|Default|SetisLocallyValid false)
                (else
                  _))
              (else
                _))))
        (else
          (Playlist|BindEventtoOnCurrentPlaylistLoaded _asfort_game_state_athena (delegate Custom|PlaylistLoaded)))))
    (:then_1
      ;[1 1 1] Set up Emissive Mat for Sign
      (bind _returnvalue (Rendering|Material|CreateDynamicMaterialInstance (Variables|Default|GetS_NomsGrocery_Sign) 0 0 "None" :!class /Script/Engine.PrimitiveComponent))
      (bind _output_get (Variables|Default|SetNOMS_EmissiveControl _returnvalue))
      (Rendering|Material|SetVectorParameterValue _output_get "Emissive Multiplier" "(R=0.000000,G=0.000000,B=0.000000,A=1.000000)" :!class /Script/Engine.MaterialInstanceDynamic)
      (bind _returnvalue_2 (Rendering|Material|CreateDynamicMaterialInstance (Variables|Default|GetSM_N_SignPiece) 0 0 "None" :!class /Script/Engine.PrimitiveComponent))
      (bind _output_get_2 (Variables|Default|SetN_EmissiveControl _returnvalue_2))
      (Rendering|Material|SetVectorParameterValue _output_get_2 "Emissive Multiplier" "(R=0.000000,G=0.000000,B=0.000000,A=1.000000)" :!class /Script/Engine.MaterialInstanceDynamic)
      (bind _returnvalue_3 (Rendering|Material|CreateDynamicMaterialInstance (Variables|Default|GetSM_O_SignPiece) 0 0 "None" :!class /Script/Engine.PrimitiveComponent))
      (bind _output_get_3 (Variables|Default|SetO_EmissiveControl _returnvalue_3))
      (Rendering|Material|SetVectorParameterValue _output_get_3 "Emissive Multiplier" "(R=0.000000,G=0.000000,B=0.000000,A=1.000000)" :!class /Script/Engine.MaterialInstanceDynamic)
      (bind _returnvalue_4 (Rendering|Material|CreateDynamicMaterialInstance (Variables|Default|GetSM_M_SignPiece) 0 0 "None" :!class /Script/Engine.PrimitiveComponent))
      (bind _output_get_4 (Variables|Default|SetM_EmissiveControl _returnvalue_4))
      (Rendering|Material|SetVectorParameterValue _output_get_4 "Emissive Multiplier" "(R=0.000000,G=0.000000,B=0.000000,A=1.000000)" :!class /Script/Engine.MaterialInstanceDynamic)
      (bind _returnvalue_5 (Rendering|Material|CreateDynamicMaterialInstance (Variables|Default|GetSM_S_SignPiece) 0 0 "None" :!class /Script/Engine.PrimitiveComponent))
      (bind _output_get_5 (Variables|Default|SetS_EmissiveControl _returnvalue_5))
      (Rendering|Material|SetVectorParameterValue _output_get_5 "Emissive Multiplier" "(R=0.000000,G=0.000000,B=0.000000,A=1.000000)" :!class /Script/Engine.MaterialInstanceDynamic))))

(event Custom|PlaylistLoaded
  (params (PlaylistName :type Name) (PlaylistContextTags :type Gameplay%20Tag%20Container%20Structure%20%28by%20ref%29))
  ;[1 1 1] Set Visibility of Sign Letters Depending on Quest Progress
  (Networking|SwitchHasAuthority
    (:Remote
      (bind (_questisvalid _questcompleted) (Default|AthenaCheckQuestandObjectiveCompletion (Utilities|Array|Get(acopy) (Fortnite|GetLocalFortPlayerControllers true)) (Variables|QuestStuff|GetQuestItem_O) (Variables|QuestStuff|GetObjBackendName_O)))
      (if _questcompleted
        (Development|SetHiddeninGame (Variables|Default|GetSM_O_SignPiece) false false)
        (bind (_questisvalid_2 _questcompleted_2) (Default|AthenaCheckQuestandObjectiveCompletion (Utilities|Array|Get(acopy) (Fortnite|GetLocalFortPlayerControllers true)) (Variables|QuestStuff|GetQuestItem_S) (Variables|QuestStuff|GetObjBackendName_S)))
        (if _questcompleted_2
          (Development|SetHiddeninGame (Variables|Default|GetSM_S_SignPiece) false false)
          (bind (_questisvalid_3 _questcompleted_3) (Default|AthenaCheckQuestandObjectiveCompletion (Utilities|Array|Get(acopy) (Fortnite|GetLocalFortPlayerControllers true)) (Variables|QuestStuff|GetQuestItem_M) (Variables|QuestStuff|GetObjBackendName_M)))
          (if _questcompleted_3
            (Development|SetHiddeninGame (Variables|Default|GetSM_M_SignPiece) false false)
            (bind (_questisvalid_4 _questcompleted_4) (Default|AthenaCheckQuestandObjectiveCompletion (Utilities|Array|Get(acopy) (Fortnite|GetLocalFortPlayerControllers true)) (Variables|QuestStuff|GetQuestItem_N) (Variables|QuestStuff|GetObjBackendName_N)))
            (if _questcompleted_4
              (Development|SetHiddeninGame (Variables|Default|GetSM_N_SignPiece) false false)
              (CallFunction|TestifFullSignVisible)
              (else
                _))
            (else
              _))
          (Variables|Default|SetisLocallyValid false)
          (else
            _))
        (else
          _)))))

(event Custom|Test%20If%20Full%20Sign%20Visible
  (bind _returnvalue (Fortnite|GetLocalFortPlayerControllers true))
  (bind _output (Utilities|Array|Get(acopy) _returnvalue))
  (bind _questitem_lightup (Variables|QuestStuff|GetQuestItem_LightUp))
  ;[1 1 1] Decide if Object should be visible to player or not depending on if they have the quest and objective or not
  (Development|SetHiddeninGame (Variables|Default|GetS_NomsGrocery_Sign) false false)
  (Development|SetHiddeninGame (Variables|Default|GetSM_N_SignPiece) true false)
  (bind (_questisvalid _questcompleted _objectivecompleted) (Default|AthenaCheckQuestandObjectiveCompletion _output _questitem_lightup (Variables|QuestStuff|GetObjBackendName_N)))
  (if _questcompleted
    (Rendering|Material|SetVectorParameterValue (Variables|Default|GetNOMS_EmissiveControl) "Emissive Multiplier" "(R=2.000000,G=2.000000,B=2.000000,A=1.000000)" :!class /Script/Engine.MaterialInstanceDynamic)
    (elif (and _questisvalid (not _objectivecompleted))
      (if (Quest|CanProgressQuestinCurrentZone (Quests|GetQuestManager _output "Invalid") _questitem_lightup)
        (Variables|Default|SetisLocallyValid true)
        (CallFunction|QuestInitializedandValid)
        ;[1 1 1] Bind Event for When Quest Gets Updated
        (bind _asfort_player_controller_athena (Utilities|Casting|CastToFortPlayerControllerAthena (Utilities|Array|Get(acopy) (Fortnite|GetLocalFortPlayerControllers true))))
        (Quests|BindEventtoOnQuestsUpdated (Quests|GetQuestManager _asfort_player_controller_athena "Invalid") (EventDispatchers|CreateEvent :!event Custom|ParentQuestUpdated))
        (else
          (Variables|Default|SetisLocallyValid false)))
      (else
        _))))

(event Custom|QuestInitializedAndValid)

(event Custom|ParentQuestUpdated
  (Utilities|FlowControl|Sequence
    (:then_0
      ;[1 1 1] Tell Children of Quest Being Updated
      (if (Variables|Default|GetListenForAllQuestUpdates)
        (CallFunction|QuestUpdatedfromAnywhere)))
    (:then_1
      ;[1 1 1] Make Interactable Non-Locally Valid
      (if (Variables|Default|GetisLocallyValid)
        (bind (_questisvalid _questcompleted _objectivecompleted) (Default|AthenaCheckQuestandObjectiveCompletion (Utilities|Array|Get(acopy) (Fortnite|GetLocalFortPlayerControllers true)) (Variables|QuestStuff|GetQuestItem_LightUp) (Variables|QuestStuff|GetObjBackendName_LightUp)))
        (if (or _questcompleted _objectivecompleted)
          (CallFunction|PlaySuccessFX)
          (Variables|Default|SetisLocallyValid false)
          ;[1 1 1] UnBind Event for When Quest Gets Updated
          (bind _asfort_player_controller_athena (Utilities|Casting|CastToFortPlayerControllerAthena (Utilities|Array|Get(acopy) (Fortnite|GetLocalFortPlayerControllers true))))
          (Quests|UnbindEventfromOnQuestsUpdated (Quests|GetQuestManager _asfort_player_controller_athena "Invalid") (EventDispatchers|CreateEvent :!event Custom|ParentQuestUpdated)))))))

(event Custom|QuestUpdatedFromAnywhere)

(event Custom|PlaySuccessFX
  (bind _signlightup_sound_cue (Variables|Default|GetSignLightUp_Sound_Cue))
  (bind _successsound (Variables|Default|GetSuccessSound))
  (bind _successfx (Variables|Default|GetSuccessFX))
  ;[1 1 1] Light Up Sign
  (Components|Timeline|SetPlayRate (Variables|Prop_NOMS_FullSign_Visit|GetSignLightTimeline) 2.0)
  (Audio|Components|Audio|Play _signlightup_sound_cue 0.0)
  (bind (_direction _lightuptime) (Timeline|SignLightTimeline PlayFromStart)
    (:Update
      (Rendering|Material|SetVectorParameterValue (Variables|Default|GetNOMS_EmissiveControl) "Emissive Multiplier" (Math|Conversions|ToLinearColor(Vector) (Math|Vector|Lerp(Vector) "0, 0, 0" "2.000000,2.000000,2.000000" _lightuptime)) :!class /Script/Engine.MaterialInstanceDynamic))
    (:Finished
      (Audio|Components|Audio|Stop _signlightup_sound_cue)
      ;[1 1 1] Play Succes FX
      (Utilities|IsValid _successsound
        (:Is%20Valid
          (Audio|SpawnSoundatLocation _successsound (Transformation|GetActorLocation)))
        (:Is%20Not%20Valid
          _))
      (Utilities|IsValid _successfx
        (:Is%20Valid
          (Effects|SpawnFXSystemAttached _successfx (Variables|/Script/Engine.Character|GetMesh (Game|Player|GetPlayerPawn (Utilities|Array|Get(acopy) (Fortnite|GetLocalFortPlayerControllers true))))))
        (:Is%20Not%20Valid
          _)))))

(fn ConstructionScript)
