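; Base Blueprint for NOMS quest interactable props, which sets up a sign mesh with its emissive multiplier zeroed.
; On BeginPlay, after the parent call, it creates a dynamic material instance on the sign mesh's first slot and stores it in LocalSignMat.
; It then sets that instance's Emissive Multiplier vector to black.
; The construction script only calls the parent, and the overlap and tick events are disabled stubs.
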(event EventBeginPlay
  (|CallParentFunction)
  (bind _returnvalue (Rendering|Material|CreateDynamicMaterialInstance (Variables|Default|GetSignMesh) 0 0 "None" :!class /Script/Engine.PrimitiveComponent))
  (bind _output_get (Variables|Default|SetLocalSignMat _returnvalue))
  (Rendering|Material|SetVectorParameterValue _output_get "Emissive Multiplier" "(R=0.000000,G=0.000000,B=0.000000,A=1.000000)" :!class /Script/Engine.MaterialInstanceDynamic))

(event Collision|EventActorBeginOverlap
  (params (OtherActor :type Actor%20Object%20Reference))
  (|CallParentFunction OtherActor) ; This node is disabled and will not be called.\nDrag off pins to build functionality.
)

(event EventTick
  (params (DeltaSeconds :type Float%20%28single-precision%29))
  (|CallParentFunction DeltaSeconds) ; This node is disabled and will not be called.\nDrag off pins to build functionality.
)

(fn ConstructionScript
  (|CallParentFunction))
